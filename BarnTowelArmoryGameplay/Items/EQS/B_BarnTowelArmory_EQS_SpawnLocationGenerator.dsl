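; Custom EQS generator that supplies candidate actors for choosing where the BarnTowelArmory POI can spawn.
; It fetches the Athena game state and its BarnTowelArmory game state component, then reads that component's available registered landing locations.
; Each valid landing location is added as a generated actor, with validity checks at every step.
; The supplied context actors are not used.
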(fn DoItemGenerationFromActors
  (params (ContextActors :type Array%20of%20Actor%20Object%20References))
  (bind _returnvalue (Fortnite|Athena|GameState|GetGameStateAthena))
  (bind _returnvalue_2 (Actor|GetComponentbyClass _returnvalue "/BarnTowelArmoryGameplay/B_MMObj_BarnTowelArmory_GameStateComponet.B_MMObj_BarnTowelArmory_GameStateComponet_C"))
  ;[0.02874 0.1159 0.1302] DoItemGeneration: Get the various locations where the POI can spawn.
  (Utilities|FlowControl|Sequence
    (:then_0
      (Utilities|IsValid _returnvalue
        (:Is%20Valid
          (Utilities|IsValid _returnvalue_2
            (:Is%20Valid
              (for _array_element (Variables|/BarnTowelArmoryGameplay/B_MMObj_BarnTowelArmory_GameStateComponet.B_MMObj_BarnTowelArmory_GameStateComponet_C|GetAvailableRegisteredLandingLocations _returnvalue_2)
                (Utilities|IsValid _array_element
                  (:Is%20Valid
                    (EQS|AddGeneratedActor _array_element)))))))))
    (:then_1
      (return))))
