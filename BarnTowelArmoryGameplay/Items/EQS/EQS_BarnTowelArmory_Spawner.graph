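# Environment query for the BarnTowelArmory spawner that ranks spawn location actors by their position relative to the safe zone.
# A custom Blueprint generator supplies the actors, and a filter tests them against the Athena safe zone.
# Two 2D distance tests score them linearly against the white circle center and the next white circle center.
# Both scoring factors bind to scalable-float providers, with negative defaults that favor closer locations.
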EnvQuery /BarnTowelArmoryGameplay/Items/EQS/EQS_BarnTowelArmory_Spawner
options [{
  generator {
   name=B_BarnTowelArmory_EQS_SpawnLocationGenerator_C_1
   class=B_BarnTowelArmory_EQS_SpawnLocationGenerator_C
   properties {
    bAutoSortTests=true
    bCanRunAsync=false
    context=&EnvQueryContext_Querier
    envQueryResultNormalizationOption=Default
    generatedItemType=/Script/AIModule.EnvQueryItemType_Actor
    generatorsActionDescription=""
    optionName=""
   }
  }
  tests [
   FortQueryTest_InsideAthenaSafeZone_0 {class=FortQueryTest_InsideAthenaSafeZone, properties {testPurpose=Filter}}
   EnvQueryTest_Distance_0 {
    class=EnvQueryTest_Distance
    properties {
     distanceTo=&/EQSContext_BarnTowelArmory_WhiteCircleSafeZoneCenter.EQSContext_BarnTowelArmory_WhiteCircleSafeZoneCenter_C
     filterType=Range
     floatValueMin=0
     scoringEquation=Linear
     testComment=""
     testMode=Distance2D
     testPurpose=Score
     scoringFactor {
      dataBinding=&self:EnvQueryTest_Distance_0.FortAIDataProvider_ScalableFloat_0
      dataField=FloatValue
      defaultValue=-0.2
     }
    }
   }
   EnvQueryTest_Distance_1 {
    class=EnvQueryTest_Distance
    properties {
     distanceTo=&/EQSContext_BarnTowelArmory_NextWhiteCircleSafeZoneCenter.EQSContext_BarnTowelArmory_NextWhiteCircleSafeZoneCenter_C
     filterType=Range
     floatValueMin=0
     scoringEquation=Linear
     testComment=""
     testMode=Distance2D
     testPurpose=Score
     scoringFactor {
      dataBinding=&self:EnvQueryTest_Distance_1.FortAIDataProvider_ScalableFloat_0
      dataField=FloatValue
      defaultValue=-0.2
     }
    }
   }
  ]
 }]
