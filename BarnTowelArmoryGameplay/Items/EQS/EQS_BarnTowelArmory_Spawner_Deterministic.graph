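# EnvQuery that finds spawn locations for a BarnTowelArmory spawner, in a deterministic variant.
# Its single option runs the Blueprint generator B_BarnTowelArmory_EQS_SpawnLocationGenerator around the querier context, producing actor items.
# The query has no tests, so results come straight from the generator without scoring or filtering, and the generator does not run asynchronously.
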EnvQuery /BarnTowelArmoryGameplay/Items/EQS/EQS_BarnTowelArmory_Spawner_Deterministic
options [{
  generator {
   name=B_BarnTowelArmory_EQS_SpawnLocationGenerator_C_1
   class=B_BarnTowelArmory_EQS_SpawnLocationGenerator_C
   properties {
    bAutoSortTests=true
    bCanRunAsync=false
    context=&EnvQueryContext_Querier
    envQueryResultNormalizationOption=Default
    generatedItemType=/Script/AIModule.EnvQueryItemType_Actor
    generatorsActionDescription=""
    optionName=""
   }
  }
  tests []
 }]
